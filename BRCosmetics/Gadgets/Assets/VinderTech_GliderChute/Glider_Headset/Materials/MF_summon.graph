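# Material function that adds an animated summon dissolve to the VinderTech headset glider via opacity mask and emissive.
# A "summon amount" parameter drives an expanding or shrinking sphere mask, with panning noise distorting the preskinned local position input for a ragged edge.
# Leg-mask texture R and G channels gate parts at staggered thresholds.
# Offset sphere masks form two tinted glow bands added to the incoming material's emissive.
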MaterialFunction /BRCosmetics/Gadgets/Assets/VinderTech_GliderChute/Glider_Headset/Materials/MF_summon
graph {nodes [
  n1 = GetMaterialAttributes(None=FunctionInput() {inputName=Material, inputType=FunctionInput_MaterialAttributes}) {attributeGetTypes [
    #1
    #2
    #3
    #4
    #5
    #6
   ]}
  n2 = SetMaterialAttributes!(
   MaterialAttributes=@n1.MaterialAttributes
   EmissiveColor=Add(A=@n1.EmissiveColor, B=NamedRerouteUsage())
   OpacityMask=NamedRerouteUsage()
   WorldPositionOffset=@n1.MaterialAttributes
   BaseColor=@n1.MaterialAttributes
   Roughness=@n1.MaterialAttributes
   Normal=@n1.MaterialAttributes
   Metallic=@n1.MaterialAttributes
  ) {attributeSetTypes [
    #1
    #7
    #2
    #3
    #4
    #5
    #6
   ]}
  FunctionOutput(None=@n2)
  n3 = SphereMask(A=NamedRerouteUsage(), B=Constant3Vector() {constant=<rgba 0 -150 0 1>}, Radius=NamedRerouteUsage())
  NamedRerouteDeclaration(None=@n3) {name=OpacityMaskNO Leg, nodeColor=<rgba 0 1 0.6719 1>}
  n4 = TextureCoordinate()
  n5 = ScalarParameter() {defaultValue=10, group=FadeAway, parameterName=NoiseTiling}
  n6 = Constant3Vector() {constant=<rgba -0.9 -0.05 0 1>}
  n7 = TextureSampleParameter2D(UVs=Panner(Coordinate=Multiply(A=@n4, B=@n5) {constB=7}, Speed=ComponentMask(None=@n6) {g=true, r=true}) {speedX=0.5, speedY=0.05}) {group=FadeAway, parameterName=FadeAwayNoise, texture=/Game/ContentCreationTools/ObjectPlacement/Textures/T_BlankWhite_01}
  n8 = ScalarParameter() {defaultValue=20, group=FadeAway, parameterName=NoiseIndensity}
  NamedRerouteDeclaration(None=Add(A=Multiply(A=@n7.RGB, B=@n8) {constB=30}, B=FunctionInput() {inputName=PreskinLocalPosition})) {name=SphereMaskInputA, nodeColor=<rgba 0 0.7656 1 1>}
  n9 = ScalarParameter() {group=FadeAway, parameterName=summon amount, sliderMax=1}
  NamedRerouteDeclaration(None=Multiply(A=OneMinus(None=@n9)) {constB=2}) {name=Tran animation, nodeColor=<rgba 0.5938 1 0 1>}
  n10 = NamedRerouteUsage()
  n11 = Subtract(A=NamedRerouteUsage()) {constB=0.5}
  n12 = Constant() {r=2.4967}
  n13 = LinearInterpolate(Alpha=Saturate(None=Multiply(A=@n11, B=@n12) {constB=1.1}))
  n14 = Constant() {r=1}
  n15 = Saturate(None=Add(
    A=TextureSample() {texture=/BRCosmetics/Gadgets/Assets/VinderTech_GliderChute/Glider_Headset/Textures/T_M_MED_headset_legmask}.R
    B=If(A=@n10, "A > B"=@n13, "A == B"=@n13, "A < B"=@n14) {constB=0.5}
   ))
  n16 = Subtract(A=NamedRerouteUsage()) {constB=0.2}
  n17 = Constant() {r=2}
  n18 = LinearInterpolate(Alpha=Saturate(None=Multiply(A=@n16, B=@n17) {constB=1.1}))
  n19 = Add(
   A=TextureSample() {texture=&T_M_MED_headset_legmask}.G
   B=If(A=NamedRerouteUsage(), "A > B"=@n18, "A == B"=@n18, "A < B"=Constant() {r=1}) {constB=0.2}
  )
  NamedRerouteDeclaration(None=Multiply(A=Multiply(A=@n3, B=@n15), B=Saturate(None=@n19))) {name=OpacityMask, nodeColor=<rgba 1 0.0469 0 1>}
  n20 = NamedRerouteUsage()
  n21 = NamedRerouteUsage()
  n22 = Subtract(
   A=@n20
   B=SphereMask(
    A=NamedRerouteUsage()
    B=Constant3Vector() {constant=<rgba 0 -150 0 1>}
    Radius=Subtract(A=@n21) {constB=0.1}
    Hardness=ScalarParameter() {defaultValue=0.99, group=FadeAway, parameterName=EdgeWidth}
   ) {hardnessPercent=99}
  )
  n23 = VectorParameter() {group=FadeAway, parameterName=GlowColorEdge, defaultValue=<rgba 7 0 0.5457 1>}
  n24 = NamedRerouteUsage()
  n25 = Constant3Vector() {constant=<rgba 0 -150 0 1>}
  n26 = Subtract(A=NamedRerouteUsage(), B=Constant() {r=0.53})
  n27 = Constant() {r=0.3}
  n28 = Power(Base=Subtract(A=@n20, B=SphereMask(A=@n24, B=@n25, Radius=@n26, Hardness=@n27) {hardnessPercent=95})) {constExponent=10}
  n29 = VectorParameter() {group=FadeAway, parameterName=GlowColor, defaultValue=<rgba 7 0 0.5457 1>}
  NamedRerouteDeclaration(None=Add(A=Multiply(A=@n22, B=@n23.RGB), B=Multiply(A=@n28, B=@n29.RGB))) {name=Emissive, nodeColor=<rgba 0 0.1328 1 1>}
  n30 = Multiply(A=NamedRerouteUsage(), B=Constant() {r=0.6})
  NamedRerouteDeclaration(None=LinearInterpolate(Alpha=Saturate(None=@n30)) {constB=1200}) {name=Spheremask_radius, nodeColor=<rgba 0.9453 0 1 1>}
 ]}
