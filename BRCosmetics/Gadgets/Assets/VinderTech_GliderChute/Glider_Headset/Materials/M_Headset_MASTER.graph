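# Master material for the VinderTech glider chute headset, a masked skeletal-mesh surface built on material attributes with Default Lit and Subsurface Profile shading.
# A shared character PreFX function feeds a chain of thin-film, metals LUT and silk layers, each masked by a channel of the SkinFX_Mask texture.
# MF_summon follows, then the character PostFX and cinematic PostFX functions.
# Thin-film and silk mask channels are selectable parameters, while metals always uses red.
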Material /BRCosmetics/Gadgets/Assets/VinderTech_GliderChute/Glider_Headset/Materials/M_Headset_MASTER
graph {
 properties {
  bAutomaticallySetUsageInEditor=false
  bUseMaterialAttributes=true
  bUsedWithMeshDeformer=true
  bUsedWithMorphTargets=true
  bUsedWithSkeletalMesh=true
  blendMode=BLEND_Masked
  pixelDepthOffsetMode=PDOM_Legacy
  shadingModel=MSM_FromMaterialExpression
  subsurfaceProfile=&SS_HeroSkin_02
  usedShadingModels=Default Lit | Subsurface Profile
 }
 nodes [
  n1 = MaterialFunctionCall() {materialFunction=&BaseFN_CharacterFunction_PreFX}
  n2 = TextureSample(Tex=NamedRerouteDeclaration(None=TextureObjectParameter() {group=Skin FX, parameterName=SkinFX_Mask}) {name=SkinFX_Mask, nodeColor=<rgba 0.9842 0 1 1>})
  n3 = StaticComponentMaskParameter(None=@n2.RGB) {defaultR=true, group=Skin FX, parameterName=ThinFilm_Channel}
  n4 = MaterialFunctionCall!(
   Alpha=ChannelMaskParameter(None=@n2.RGB) {group=Skin FX, parameterName=Silk Channel}.R
   Material=MaterialFunctionCall(
    Mask=ComponentMask(None=@n2.RGB) {r=true}
    Materials=MaterialFunctionCall!(Alpha=@n3, Material Attributes=@n1.Result) {materialFunction=/Game/Weapons/WeaponSkins/MaterialFunction/ThinFilmFunction}.Result
   ) {materialFunction=&MF_MetalsLUT}.Result
  ) {materialFunction=/Game/Weapons/WeaponSkins/MaterialFunction/MF_Silk}
  n5 = MaterialFunctionCall(
   AO=@n1.AO
   Distance Blend=@n1.AO
   Fresnel=@n1.AO
   Gmap=@n1.AO
   Pre Skinned Local Position=@n1.AO
   Material=MaterialFunctionCall!(PreskinLocalPosition=@n1.Pre skinned Local Position, Material=GetMaterialAttributes(None=@n4.Result).MaterialAttributes) {materialFunction=&/MF_summon}.Result
  ) {materialFunction=&BaseFN_CharacterFunction_PostFX}
  MaterialFunctionCall(In=@n5.Result) {materialFunction=&CINE_MF_CharPostFX}
 ]
}
